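# Niagara shoe effect for the SafariGnome SlamCraze skin, attached to the skeletal mesh and tinted by time of day.
# MeshTrails bursts four persistent meshes across two bones, shrinking and fading them as RunningLerp rises, with a delayed front-end fade-in.
# Two mirrored ribbon emitters spawn only while running, scaling opacity erosion by sampled bone velocity.
# Foot-hidden suppresses all emitters, while skydiving and HideFootFX stop only the ribbons.
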NiagaraSystem /BRCosmetics/Characters/Player/Male/Medium/Bodies/M_MED_SafariGnome/Skins/SlamCraze/FX/NS_Character_SlamCraze_Shoes_Idle
dependencies {
 usedDataInterfaces [&NiagaraDataInterfaceCurve, &NiagaraDataInterfaceColorCurve]
 usedDynamicInputs [
  &SetBoolByFloatComparison
  &Multiply_Vector
  &ClampFloat
  &Divide_Float
  &ColorFromCurve
 ]
 usedModules [
  &self:SystemSpawnScript.SystemScriptSource.SystemScriptGraph.NiagaraNodeAssignment_2.SetVariables_#1
  &SystemState
  &self:SystemSpawnScript.SystemScriptSource.SystemScriptGraph.NiagaraNodeAssignment_0.SetVariables_#2
  &self:SystemSpawnScript.SystemScriptSource.SystemScriptGraph.NiagaraNodeAssignment_3.SetVariables_#3
  &self:SystemSpawnScript.SystemScriptSource.SystemScriptGraph.NiagaraNodeAssignment_1.SetVariables_#4
 ]
}
emitters {
 FN_AttachedRibbon {
  emitterUpdateScript [
   &EmitterState()
   &FN_SpawnRibbon_v2(
    Number of Ribbons:NiagaraInt32=&MakeCustomIntFromBool(
     Bool=$User.bisFootHidden
     True INT=0
     False INT=&MakeCustomIntFromBool(Bool=$User.IsSkydiving, True INT=0, False INT=2)
    )
    Spawn Mode=::NewEnumerator0
    SpawnRate:NiagaraFloat=&MakeCustomFloatFromBool(
     Bool=$User.bisFootHidden
     True Float=0
     False Float=&MakeCustomFloatFromBool(
      Bool=$User.IsSkydiving
      True Float=0
      False Float=&MakeCustomFloatFromBool(
       Bool=$User.HideFootFX
       True Float=0
       False Float=&MakeCustomFloatFromBool(Bool=$System.isRunning, True Float=30, False Float=0)
      )
     )
    )
   )
  ]
  particleSpawnScript [
   /Game/Effects/Niagara/Modules/Ribbons/FN_InitializeRibbon_v4(
    Custom Offset=true
    Enable Skeletal Position=true
    Lifetime=0.1
    Local Vector Offset:Vector3f=<1 0 -1>
    Offset Amount=3.4575
    Ribbon Width=25
    SkeletalMesh:NiagaraDataInterfaceSkeletalMesh=$System.Skeletal Mesh
   )
   &FN_FlipRibbonV()
   &AddVelocity(Velocity=<0 0 100>) {enabled=false}
  ]
  particleUpdateScript [
   &InheritSourceMovement(
    Applied Position Scale=&VectorFromFloat(Value=&FloatFromCurve(FloatCurve=DataInterface() {propertyValues {Curve {keys [
          {interpMode=RCIM_Cubic, tangentMode=RCTM_User, time=0, value=0.6394}
          {arriveTangent=-0.0657, interpMode=RCIM_Cubic, leaveTangent=-0.0657, tangentMode=RCTM_User, time=1, value=0}
         ]}}}))
    ApplyForce=false
    ApplyPosition=true
    ApplyPositionFalloff=false
   )
   /SetVariables(Particles.RibbonWidth=&FloatFromCurve(
     FloatCurve=DataInterface() {propertyValues {Curve {keys [{arriveTangent=-0.3053, time=0, value=1}, {arriveTangent=-0.5499, time=0.3062, value=0.5}]}}}
     CurveIndex=$Particles.RibbonU0Override
     Scale Curve=&MakeCustomFloatFromBool(Bool=$User.bisFootHidden, True Float=0, False Float=3)
    )) {isSetParameters=true, module=SetVariables_#5}
   &SolveForcesAndVelocity(Clamp Velocity=false, Limit Acceleration=false)
   /Game/Effects/Niagara/Modules/Ribbons/FN_UpdateRibbon_v4(Enable Skeletal Position=true, SkeletalMesh:NiagaraDataInterfaceSkeletalMesh=$System.Skeletal Mesh)
   /SetVariables(Particles.AlphaScale:NiagaraFloat=&RemapRange(
     Input Value To Remap=&VectorLength(Vector=$Particles.FN_UpdateRibbon_v4.SampledVelocity)
     Input Min=100
     Input Max=300
     Output Min=0
     Output Max=1
    )) {isSetParameters=true, module=SetVariables_#6}
   &FN_RibbonMaterial(
    Color=&MultiplyLinearColor(
     Linear Color=/Niagara/DynamicInputs/LinearColor/MakeLinearColorFromVector4(Vector:Vector4f=/Niagara/DynamicInputs/Multiply/Multiply_Vector4_ByFloat(
       Vector4:Vector4f=&MakeVector4_FromLinearColor(LinearColor=<rgba 0.5 0.4 0.2 1.7>)
       Float:NiagaraFloat=&FNTimeOfDayScalar(Night Scale=0.8, Morning Scale=0.9, Evening Scale=0.85)
      ))
     Scale Factor=&MakeVector4_FromLinearColor(LinearColor=$System.ToDColor)
    )
    Opacity Erosion:NiagaraFloat=&FloatFromCurve(
     FloatCurve=DataInterface() {propertyValues {Curve {
        defaultValue=0
        keys [{arriveTangent=1.1228, interpMode=RCIM_Cubic, leaveTangent=1.1228, tangentMode=RCTM_User, time=0, value=0}, <at 1 1 RCIM_Cubic>]
       }}}
     Scale Curve=$Particles.AlphaScale
    )
    Opacity Pan U Speed=5
    Opacity Power=0.5
    Scale Alpha=1.5
   )
   &DynamicMaterialParameters(Module.Index0Float0Write=true, Module.Index0Float1Write=true, Module.Index0Float2Write=true, Module.Index0Float3Write=true) {module=DynamicMaterialParameters001}
  ]
  rendererClasses [&NiagaraRibbonRendererProperties]
  renderers [{rendererClass=&NiagaraRibbonRendererProperties}]
 }
 FN_AttachedRibbon001 {
  emitterUpdateScript [
   &EmitterState()
   &FN_SpawnRibbon_v2(
    Number of Ribbons:NiagaraInt32=&MakeCustomIntFromBool(
     Bool=$User.bisFootHidden
     True INT=0
     False INT=&MakeCustomIntFromBool(Bool=$User.IsSkydiving, True INT=0, False INT=2)
    )
    Spawn Mode=::NewEnumerator0
    SpawnRate:NiagaraFloat=&MakeCustomFloatFromBool(
     Bool=$User.bisFootHidden
     True Float=0
     False Float=&MakeCustomFloatFromBool(
      Bool=$User.IsSkydiving
      True Float=0
      False Float=&MakeCustomFloatFromBool(
       Bool=$User.HideFootFX
       True Float=0
       False Float=&MakeCustomFloatFromBool(Bool=$System.isRunning, True Float=30, False Float=0)
      )
     )
    )
   )
  ]
  particleSpawnScript [
   &FN_InitializeRibbon_v4(
    Custom Offset=true
    Enable Skeletal Position=true
    Lifetime=0.1
    Local Vector Offset:Vector3f=<1 0 1>
    Offset Amount=-3.8341
    Ribbon Width=25
    SkeletalMesh:NiagaraDataInterfaceSkeletalMesh=$System.Skeletal Mesh
   )
   &FN_FlipRibbonV()
   &AddVelocity(Velocity=<0 0 100>) {enabled=false}
  ]
  particleUpdateScript [
   &InheritSourceMovement(
    Applied Position Scale=&VectorFromFloat(Value=&FloatFromCurve(FloatCurve=DataInterface() {propertyValues {Curve {keys [
          {interpMode=RCIM_Cubic, tangentMode=RCTM_User, time=0, value=0.6394}
          {arriveTangent=-0.0657, interpMode=RCIM_Cubic, leaveTangent=-0.0657, tangentMode=RCTM_User, time=1, value=0}
         ]}}}))
    ApplyForce=false
    ApplyPosition=true
    ApplyPositionFalloff=false
   )
   /SetVariables(Particles.RibbonWidth=&FloatFromCurve(
     FloatCurve=DataInterface() {propertyValues {Curve {keys [{arriveTangent=-0.3053, time=0, value=1}, {arriveTangent=-0.5499, time=0.3062, value=0.5}]}}}
     CurveIndex=$Particles.RibbonU0Override
     Scale Curve=&MakeCustomFloatFromBool(Bool=$User.bisFootHidden, True Float=0, False Float=3)
    )) {isSetParameters=true, module=SetVariables_#5}
   &SolveForcesAndVelocity(Clamp Velocity=false, Limit Acceleration=false)
   &FN_UpdateRibbon_v4(Enable Skeletal Position=true, SkeletalMesh:NiagaraDataInterfaceSkeletalMesh=$System.Skeletal Mesh)
   /SetVariables(Particles.AlphaScale:NiagaraFloat=&RemapRange(
     Input Value To Remap=&VectorLength(Vector=$Particles.FN_UpdateRibbon_v4.SampledVelocity)
     Input Min=100
     Input Max=300
     Output Min=0
     Output Max=1
    )) {isSetParameters=true, module=SetVariables_#6}
   &FN_RibbonMaterial(
    Color=&MultiplyLinearColor(
     Linear Color=&MakeLinearColorFromVector4(Vector:Vector4f=&Multiply_Vector4_ByFloat(
       Vector4:Vector4f=&MakeVector4_FromLinearColor(LinearColor=<rgba 0.5 0.4 0.2 1.7>)
       Float:NiagaraFloat=&FNTimeOfDayScalar(Night Scale=0.8, Morning Scale=0.9, Evening Scale=0.85)
      ))
     Scale Factor=&MakeVector4_FromLinearColor(LinearColor=$System.ToDColor)
    )
    Opacity Erosion:NiagaraFloat=&FloatFromCurve(
     FloatCurve=DataInterface() {propertyValues {Curve {
        defaultValue=0
        keys [{arriveTangent=1.1228, interpMode=RCIM_Cubic, leaveTangent=1.1228, tangentMode=RCTM_User, time=0, value=0}, <at 1 1 RCIM_Cubic>]
       }}}
     Scale Curve=$Particles.AlphaScale
    )
    Opacity Pan U Speed=5
    Opacity Power=0.5
    Scale Alpha=1.5
   )
   &DynamicMaterialParameters(Module.Index0Float0Write=true, Module.Index0Float1Write=true, Module.Index0Float2Write=true, Module.Index0Float3Write=true) {module=DynamicMaterialParameters001}
  ]
  rendererClasses [&NiagaraRibbonRendererProperties]
  renderers [{rendererClass=&NiagaraRibbonRendererProperties}]
 }
 MeshTrails {
  emitterUpdateScript [
   &EmitterState(
    Inactive Response=::NewEnumerator1
    Life Cycle Mode=::NewEnumerator1
    Loop Behavior=::NewEnumerator1
    Loop Duration Mode=::NewEnumerator1
    UseLoopDelay=false
   )
   &SpawnBurst_Instantaneous(
    Spawn Count=&MakeCustomIntFromBool(Bool=$User.bisFootHidden, True INT=0, False INT=4)
    Use Spawn Probability=false
    UseLoopCountLimit=false
   )
  ]
  particleSpawnScript [
   &FN_SetSpawnGroup()
   &InitializeParticle(
    Mesh Scale=&MergeHelper_VectorSwitch(
     0=<0.19 0.35 0.35>
     1=<0.25 0.35 0.35>
     2=<0.19 0.35 0.35>
     3=<0.25 0.35 0.35>
     Index=&SpawnGroup()
     Num=::NewEnumerator2
    )
    Mesh Scale Mode=::NewEnumerator4
    UsePositionOffset=false
    Write Material Random=true
   )
   &InitialMeshOrientation(
    Mesh Orientation Mode=::NewEnumerator2
    Orientation Vector=&MergeHelper_VectorSwitch(0=<0.2 -1 -0.07>, 1=<0.2 -1 0.05>, 2=<-0.2 1 0.07>, 3=<-0.2 1 -0.05>, Index=&SpawnGroup(), Num=::NewEnumerator2)
    Rotation=&MergeHelper_VectorSwitch(0=<-0.25 0 0>, 1=<0.25 0 0>, 2=<0.25 0 0>, 3=<-0.25 0 0>, Index=&SpawnGroup(), Num=::NewEnumerator2)
    Use Rotation Vector=true
   )
  ]
  particleUpdateScript [
   &ParticleState(Kill Particles When Lifetime Has Elapsed=false, Let Infinitely Lived Particles Die When Emitter Deactivates=true)
   &ScaleMeshSize(Scale Factor=&MakeCustomVectorFromBool(
     Bool=$User.bisFootHidden
     True Vector=<0 0 0>
     False Vector=&MakeVector(
      X=&Lerp_Float(A=0.752, B=0.3, Alpha=$System.RunningLerp)
      Y=&Lerp_Float(A=0.8, B=0.3, Alpha=$System.RunningLerp)
      Z=&Lerp_Float(A=0.8, B=0.3, Alpha=$System.RunningLerp)
     )
    ))
   &ScaleColor(
    Scale Mode=::NewEnumerator1
    Scale RGBA=&Multiply_Vector4_ByFloat(
     Vector4:Vector4f=&MakeVector4_FromLinearColor(LinearColor=<rgba 0.9 0.705 0.315 1.7>)
     Float:NiagaraFloat=&FNTimeOfDayScalar(Night Scale=0.8, Morning Scale=0.9, Evening Scale=0.85)
    )
    ScaleRGBA=true
   ) {module=ScaleColor001}
   &SkeletalMeshLocation(
    Bone / Socket Index=&MergeHelper_IntSwitch(0=0, 1=0, 2=1, 3=1, 4=?, 5=?, Index=&SpawnGroup(), Num=::NewEnumerator2)
    Bone Sampling Mode=::NewEnumerator4
    MaskSpawnGroup=false
    Sampled Position Offset=&MergeHelper_VectorSwitch(
     0=<-5.6997 3.6582 -3.9678>
     1=<-5.6501 3.2568 3.4979>
     2=<5.6205 -4.0847 4.0469>
     3=<5.8817 -4.2006 -3.4308>
     Index=&SpawnGroup()
     Num=::NewEnumerator2
    )
    Skeletal Mesh=$System.Skeletal Mesh
   )
   &InterpolateOverTime(Rate Of Change=30, Target Vector Value:NiagaraPosition=$Particles.Position, Variable Type=::NewEnumerator2)
   /SetVariables(Particles.Interpolate DY:Vector4f=&MakeFloat4FromVectorAndFloat("Vector (RGB)"=$Particles.InterpolateOverTime.InterpolateOverTime_V3_.Moving Average, "Float (Alpha)"=1)) {isSetParameters=true}
   &DynamicMaterialParameters(
    Index 0 Param 2=&MergeHelper_FloatSwitch(0=0.6, 1=0, 2=0.7, 3=0.2, Index=&SpawnGroup(), Num=::NewEnumerator2)
    Index 0 Param 3=6
    Index 0 Param 4=1
    Module.Index0Float0Write=true
    Module.Index0Float1Write=true
    Module.Index0Float2Write=true
    Module.Index0Float3Write=true
   )
   &ScaleColor(
    Scale Alpha=&Lerp_Float(
     A=&FN_FrontEndFloatModifier(
      Gameplay=1
      Front End=&FloatFromCurve(FloatCurve=DataInterface() {propertyValues {Curve {keys [<at 0.3 0>,<at 0.9 1>]}}}, CurveIndex=$System.Age)
     )
     B=0
     Alpha=$System.RunningLerp
    )
    Scale RGB=&MakeVectorFromLinearColorRGB(Color=$System.ToDColor)
    ScaleA=true
    ScaleRGB=true
   ) {module=ScaleColor002}
  ]
  rendererClasses [&NiagaraMeshRendererProperties]
  renderers [{rendererClass=&NiagaraMeshRendererProperties, rendererIndex=0}]
 }
}
summary {
 systemName=NS_Character_SlamCraze_Shoes_Idle
 emitters [
  {emitterName=MeshTrails, rendererClasses [&NiagaraMeshRendererProperties]}
  {emitterName=FN_AttachedRibbon, rendererClasses [&NiagaraRibbonRendererProperties]}
  {emitterName=FN_AttachedRibbon001, rendererClasses [&NiagaraRibbonRendererProperties]}
 ]
 userVariables [
  User.bisFootHidden {type=NiagaraBool}
  User.bIsFrontEnd {type=NiagaraBool}
  User.bIsFrontEndPreview {type=NiagaraBool}
  User.bisHeadHidden {type=NiagaraBool}
  User.HideFootFX {type=NiagaraBool}
  User.IsSkydiving {type=NiagaraBool}
  User.Mask {type=NiagaraFloat, defaultValue=0, description=""}
  User.Reactivity {type=NiagaraFloat, defaultValue=0, description=""}
  User.RibbonIDOffset {type=NiagaraInt32}
  User.SkeletalMesh {
   type=NiagaraDataInterfaceSkeletalMesh
   defaultValue {
    struct=NiagaraExt_VariableValue_DataInterface
    value {dataInterface=&self:NiagaraDataInterfaceSkeletalMesh_1}
   }
  }
  User.Variant {type=NiagaraFloat, defaultValue=0, description=""}
 ]
}
